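# Sound cue for the player recharge sound in MotoWrestle, routed to the non-critical in-game UI sound class.
# A Mixer layers a UI meter-fill wave at full volume with a quieter temp recharge wave that passes through a default-setting Modulator for variation.
# A concurrency override limits the cue to a single playing instance.
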SoundCue /MotoWrestleGameplay/Assets/Audio/IP_SC_MotoWrestle_Player_Recharge
root=Mixer_0
nodes [
 {id=Mixer_0, children [WavePlayer_0,Modulator_0], properties {inputVolume [1,0.4]}}
 {id=Modulator_0, children [WavePlayer_1]}
 {id=WavePlayer_0, properties {soundWaveAssetPtr=&/IP_MotoWrestle_UI_Meter_Fill_01}}
 {id=WavePlayer_1, properties {soundWaveAssetPtr=&/MotoWrestle_Recharge_TEMP}}
]
properties {
 bOverrideConcurrency=true
 maxDistance=2097152
 soundClassObject=&Fort_UI_InGame_NonCritical
 volumeMultiplier=1
 concurrencyOverrides {maxCount=1}
}
