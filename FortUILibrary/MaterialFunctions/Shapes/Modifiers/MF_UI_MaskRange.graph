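# MF_UI_MaskRange is a library material function for UI that turns an SDF shape or distance field into an anti-aliased mask between RangeStart and RangeEnd.
# It wraps MF_UI_ShapeOutline, passing the range width as Thickness and the sum of both bounds divided by that width as Alignment.
# An optional Smoothing input softens the edges.
# A regular polygon SDF feeds the DistanceField preview.
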MaterialFunction /FortUILibrary/MaterialFunctions/Shapes/Modifiers/MF_UI_MaskRange
graph {
 properties {
  bExposeToLibrary=true
  description=Returns an anti-aliased mask from the supplied SDF shape or distance field texture. The values between RangeStart and RangeEnd will be '1.0' while the rest of the mask will be '0.0'.
  previewMaterialDomain=MD_UI
  libraryCategoriesText [UI_Library]
 }
 nodes [
  n1 = FunctionInput() {
   description=The value of the distance field where the mask begins.
   inputName=RangeStart
   inputType=FunctionInput_Scalar
   sortPriority=1
   previewValue=<0.1 0 0 1>
  }
  n2 = FunctionInput() {
   description=The value of the distance field where the mask ends.
   inputName=RangeEnd
   inputType=FunctionInput_Scalar
   sortPriority=2
   previewValue=<0.15 0 0 1>
  }
  n3 = Add(A=@n1, B=@n2)
  n4 = Subtract(A=@n2, B=@n1)
  n5 = MaterialFunctionCall!(UVs, Radius, Sides, InnerAngle) {materialFunction=/FortUILibrary/MaterialFunctions/Shapes/SDFs/MF_SDF_RegularPolygon}
  n6 = FunctionInput() {
   bUsePreviewValueAsDefault=true
   description=Softens the resulting edges of the Mask.
   inputName=Smoothing
   inputType=FunctionInput_Scalar
   sortPriority=3
  }
  FunctionOutput(None=MaterialFunctionCall!(
    DistanceField=FunctionInput(Preview=@n5.SDF) {description=The distance field to generate a mask from., inputName=DistanceField, inputType=FunctionInput_Scalar}
    Thickness=@n4
    Alignment=Divide(A=@n3, B=@n4)
    Smoothing=@n6
   ) {materialFunction=&/MF_UI_ShapeOutline}.Outline) {description=The resulting masked shape., outputName=Mask}
 ]
}
