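# Translucent UI material for the MotoWrestle pip meter: a segmented radial arc beside an MSDF-textured container shape with soft shadows.
# CurrentPip, TotalPips and Progress drive two radial fills, separating completed pips from the cooling-down one, which a sector SDF cuts into spaced segments.
# Radially mirrored ice-crack and panning noise textures create animated icy color effects.
# PipFlashSwitch widens outline smoothing and blends in a flash layer.
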Material /MotoWrestleGameplay/UI/Materials/M_UI_Pip_MotoWrestle
graph {
 properties {bAutomaticallySetUsageInEditor=false, blendMode=BLEND_Translucent, materialDomain=MD_UI, usedShadingModels=Default Lit}
 nodes [
  n1 = NamedRerouteUsage()
  n2 = ScalarParameter() {parameterName=PipFlashSwitch}
  n3 = ScalarParameter() {parameterName=Progress}
  n4 = LinearInterpolate(
   A=Constant3Vector() {constant=<rgba 1 1 1 1>}
   B=Constant3Vector() {constant=<rgba 0.1356 0.7529 1 1>}
   Alpha=OneMinus(None=Multiply(A=@n3) {constB=1.25})
  )
  n5 = NamedRerouteUsage()
  n6 = NamedRerouteUsage()
  n7 = NamedRerouteUsage()
  n8 = LinearInterpolate(
   A=Max(A=LinearInterpolate(B=@n4, Alpha=@n5), B=LinearInterpolate(B=@n6, Alpha=@n7))
   B=VectorParameter() {parameterName=Pip Bg, defaultValue=<rgba 0.0182 0.0182 0.0182 0>}.RGB
   Alpha=NamedRerouteUsage()
  )
  n9 = NamedRerouteUsage()
  n10 = Constant3Vector() {constant=<rgba 1 1 1 1>}
  n11 = NamedRerouteUsage()
  Add(
   A=LinearInterpolate(B=NamedRerouteUsage(), Alpha=LinearInterpolate(B=@n1, Alpha=@n2))
   B=LinearInterpolate(A=LinearInterpolate(B=@n8, Alpha=@n9), B=LinearInterpolate(B=@n10, Alpha=@n11), Alpha=NamedRerouteUsage())
  )
  Max(A=NamedRerouteUsage(), B=Multiply(A=NamedRerouteUsage()) {constB=0.5})
  n12 = Add(A=NamedRerouteUsage(), B=NamedRerouteUsage()) {constB=0}
  n13 = MaterialFunctionCall(UVs=@n12, "Rotation Angle (0-1)"=Constant() {r=0.729}) {materialFunction=&CustomRotator}
  n14 = ScalarParameter() {parameterName=Start Angle, sliderMax=1, sortPriority=31}
  n15 = ScalarParameter() {defaultValue=0.91, parameterName=End Angle, sliderMax=1}
  n16 = Constant()
  n17 = ScalarParameter() {parameterName=Progress}
  n18 = Constant() {r=1}
  n19 = ScalarParameter() {defaultValue=3, parameterName=CurrentPip}
  n20 = Add(
   A=OneMinus(None=If(A=@n17, "A > B"=@n17, "A == B"=@n18, "A < B"=@n17) {desc=If for Rounding Error Bug - Equals Threshold Needed, equalsThreshold=0.01})
   B=@n19
  )
  n21 = ScalarParameter() {defaultValue=3, parameterName=TotalPips}
  n22 = MaterialFunctionCall!(
   UV=@n13.Rotated Values
   "Arc Start Angle (0-1)"=@n14
   "Arc End Angle (0-1)"=@n15
   "Fill Ratio (0-1)"=OneMinus(None=Divide(A=@n20, B=@n21))
   Fill Clockwise=@n16
  ) {materialFunction=/FortUILibrary/MaterialFunctions/Basics/MF_UI_RadialArc}
  n23 = OneMinus(None=ScalarParameter() {defaultValue=0.125, desc=Segment Spacing, parameterName="Segment Spacing ", sliderMax=1})
  n24 = SmoothStep(Value=MaterialFunctionCall!(
    UVs
    Gradient Map=@n22.Arc Mask
    Sector Count=NamedRerouteUsage()
    Sector Spacing=NamedRerouteDeclaration(None=@n23) {name=Segment Spacing, nodeColor=<rgba 0 0 0 1>}
    Fixed Spacing=Constant() {r=1}
   ) {materialFunction=/FortUILibrary/MaterialFunctions/Basics/MF_UI_Sectors}.Convergent Sectors SDF) {constMax=0.2}
  n25 = Ceil(None=@n22.Radial Fill)
  n26 = NamedRerouteDeclaration(None=@n21) {name=Segments Count, nodeColor=<rgba 1 1 1 1>}
  n27 = MaterialFunctionCall(
   UV=@n13.Rotated Values
   "Arc Start Angle (0-1)"=@n14
   "Arc End Angle (0-1)"=@n15
   "Fill Ratio (0-1)"=OneMinus(None=Divide(A=@n19, B=@n26))
   Fill Clockwise=@n16
  ) {materialFunction=&MF_UI_RadialArc}
  n28 = NamedRerouteDeclaration(None=Ceil(None=@n27.Radial Fill)) {name=CompletedPips, nodeColor=<rgba 0 1 0.3906 1>}
  n29 = Max(A=NamedRerouteDeclaration(None=Subtract(A=@n25, B=@n28)) {name=CooldownDelta, nodeColor=<rgba 1 1 1 1>}, B=@n28)
  n30 = NamedRerouteDeclaration(None=ScalarParameter() {defaultValue=0.25, parameterName=BackgroundOpacity}) {name=MeterBackground, nodeColor=<rgba 1 1 1 1>}
  n31 = MaterialFunctionCall(UVs=@n12, Size=Constant() {r=1.8}, OutlineThickness=ScalarParameter() {defaultValue=0.075, parameterName=Thickness}) {materialFunction=&MF_UI_Circle}
  n32 = Multiply(
   A=Multiply(A=Max(A=@n29, B=@n30) {constB=0.25}, B=OneMinus(None=@n24))
   B=MaterialFunctionCall(DistanceField=@n31.SDFOutline) {materialFunction=&MF_UI_AntiAliasing}.Inner
  )
  NamedRerouteDeclaration(None=@n32) {name=MeterPips, nodeColor=<rgba 1 1 1 1>}
  n33 = Constant4Vector() {constant=<rgba 0 0.1445 0.5781 1>}
  n34 = Constant4Vector() {constant=<rgba 0.8142 1 1 1>}
  n35 = ConstantBiasScale(None=TextureCoordinate()) {bias=-0.1, scale=2}
  n36 = Time()
  n37 = Multiply(A=@n36) {constB=0.0025}
  n38 = Panner(Coordinate=Abs(None=MaterialFunctionCall(UVs=@n35, "Rotation Angle (0-1)"=@n37) {materialFunction=&CustomRotator}.Rotated Values)) {speedX=0.001, speedY=0.001}
  n39 = ScalarParameter() {defaultValue=3, parameterName=TotalPips}
  n40 = TextureSample(UVs=Multiply(A=MaterialFunctionCall!(UVs=@n38, Sides=@n39) {materialFunction=/FortUILibrary/MaterialFunctions/Basics/MF_UI_UVRadialMirror}.UVs)) {samplerType=SAMPLERTYPE_Masks, texture=/Game/Environments/Asteria/Terrain/Textures/T_Asteria_Ice_Cracks_M}
  n41 = Add(A=@n36, B=Multiply(A=OneMinus(None=@n40.R)) {constB=10})
  n42 = TextureSample() {texture=/Game/Global/Textures/Default/Noise/T_UI_Ice_Noise}
  n43 = Multiply(A=@n42.R) {constB=10}
  n44 = TextureSample(UVs=Panner(Time=Add(A=@n41, B=@n43) {constB=0}) {speedY=0.02}) {texture=&T_UI_Ice_Noise}
  n45 = Sine(None=@n36) {period=2}
  n46 = Power(Base=Multiply(A=@n44.R, B=@n42.G), Exp=LinearInterpolate(Alpha=Saturate(None=@n45)) {constA=0.5, constAlpha=1, constB=0.825}) {constExponent=0.5}
  NamedRerouteDeclaration(None=MaterialFunctionCall(A=@n33.RGBA, B=Constant4Vector() {constant=<rgba 0.1017 0.792 1 1>}.RGBA, C=@n34.RGBA, Alpha=@n46) {materialFunction=&3ColorBlend}.Result) {name=IceFX, nodeColor=<rgba 1 1 1 1>}
  n47 = MaterialFunctionCall(DistanceField=@n31.SDFOutline, Smoothing=LinearInterpolate(Alpha=ScalarParameter() {parameterName=PipFlashSwitch}) {constA=0.05, constB=0.2}) {materialFunction=&MF_UI_AntiAliasing}
  n48 = Ceil(None=@n27.Arc Mask)
  NamedRerouteDeclaration(None=LinearInterpolate(A=@n32, B=Multiply(A=@n47.Inner, B=@n48)) {constAlpha=0.4}) {name=Meter+Shadow, nodeColor=<rgba 1 1 1 1>}
  n49 = Constant()
  n50 = Divide(B=ScalarParameter() {defaultValue=3, parameterName=TotalPips}) {constA=0.5}
  NamedRerouteDeclaration(None=Multiply(
    A=MaterialFunctionCall!(DistanceField=@n27.Radial Fill, RangeStart=@n49, RangeEnd=@n50, Smoothing) {materialFunction=/FortUILibrary/MaterialFunctions/Shapes/Modifiers/MF_UI_MaskRange}.Mask
    B=NamedRerouteUsage()
   )) {name=PipFlash, nodeColor=<rgba 1 1 1 1>}
  n51 = MaterialFunctionCall() {materialFunction=&GetUserInterfaceUV}
  n52 = Constant2Vector() {g=0.19, r=0.56}
  NamedRerouteDeclaration(None=Add(
    A=NamedRerouteDeclaration(None=@n51.Normalized UV) {name=UVs, nodeColor=<rgba 1 1 1 1>}
    B=NamedRerouteDeclaration(None=@n52) {name=UVShift, nodeColor=<rgba 1 1 1 1>}
   ) {constB=0}) {name=Center Point UVs, nodeColor=<rgba 1 1 1 1>}
  n53 = Add(A=NamedRerouteUsage(), B=Constant2Vector() {g=-0.172, r=-0.22})
  n54 = Constant() {r=0.6439}
  n55 = MaterialFunctionCall!(MSDF RGB=TextureSample(UVs=MaterialFunctionCall(UVs=@n53, Texture Scale=@n54) {materialFunction=&ScaleUVsByCenter}.UVs) {samplerType=SAMPLERTYPE_LinearColor, texture=&/T_UI_MSDF_MotoWrestle_Pip}.RGB) {materialFunction=/FortUILibrary/MaterialFunctions/Shapes/Modifiers/MF_UI_MSDFtoSDF}
  n56 = Constant() {r=0.45}
  n57 = MaterialFunctionCall(DistanceField=@n55.SDF, Threshold=@n56) {materialFunction=&MF_UI_AntiAliasing}
  n58 = MaterialFunctionCall(DistanceField=@n55.SDF, Threshold=Add(A=@n56) {constB=-0.15}) {materialFunction=&MF_UI_AntiAliasing}
  NamedRerouteDeclaration(None=LinearInterpolate(
    A=NamedRerouteDeclaration(None=@n57.Outer) {name=Container, nodeColor=<rgba 1 1 1 1>}
    B=NamedRerouteDeclaration(None=@n58.Outer) {name=ContainerShadow, nodeColor=<rgba 1 1 1 1>}
   ) {constAlpha=0.4}) {name=ContainerCombinedMask, nodeColor=<rgba 1 1 1 1>}
  n59 = Add(A=Add(A=@n46, B=@n40.R) {constB=0.1}, B=@n44.R)
  NamedRerouteDeclaration(None=ComponentMask(None=LinearInterpolate(A=@n34.RGBA, B=@n33.RGBA, Alpha=@n59)) {b=true, g=true, r=true}) {name=FlashFX, nodeColor=<rgba 1 1 1 1>}
 ]
}
