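# MotoWrestle weapon reticle widget, built on the recharge-stack reticle base, showing readiness and recharging ammo pips.
# Ammo changes update the pip material, playing a flash and recharge sound when the count rises.
# While the magazine isn't full, a timer polls item cooldown to drive the material's Progress parameter.
# Remaining ammo, combined with target range on range changes, plays the ready animation forward or in reverse.
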WidgetBlueprint /MotoWrestleGameplay/UI/WBP_MotoWrestle_Reticle
animations [
 Anim_IceSlide {
  binding_count=1
  length_seconds=0.5226
  object=&self:Anim_IceSlide
  sequence {
   bindings [Image_Reticle {
     id=#1
     possessed_class=Image
     tracks [{
       type=MovieSceneWidgetMaterialTrack
       sections [{
         type=MovieSceneParameterSection
         end_frame=15
         start_frame=0
         channels [InnerBracketSwitch {type=MovieSceneScriptingFloatChannel, keys [[0,0],[15.677,1]]}]
        }]
      }]
    }]
   display_rate {numerator=30}
  }
 }
 Anim_ReadyForUse {
  binding_count=1
  length_seconds=0.2462
  object=&self:Anim_ReadyForUse
  sequence {bindings [Image_Reticle {
     id=#2
     possessed_class=Image
     tracks [{
       type=MovieSceneWidgetMaterialTrack
       sections [{
         type=MovieSceneParameterSection
         end_frame=4
         start_frame=0
         channels [
          Spikes Color Switch {type=MovieSceneScriptingFloatChannel, keys [[0,0],[4.9237,1]]}
          Reticle Master Opacity {type=MovieSceneScriptingFloatChannel, keys [[0,0],[4.9233,1]]}
         ]
        }]
      }]
    }]}
 }
 Anim_PipReady {
  binding_count=1
  length_seconds=0.2667
  object=&self:Anim_PipReady
  sequence {
   bindings [Image_Reticle_Pips {
     id=#3
     possessed_class=Image
     tracks [{
       type=MovieSceneWidgetMaterialTrack
       sections [{
         type=MovieSceneParameterSection
         end_frame=8
         start_frame=0
         channels [PipFlashSwitch {type=MovieSceneScriptingFloatChannel, keys [[0,0],[1,1],[4,1],[8,0]]}]
        }]
      }]
    }]
   display_rate {numerator=30}
  }
 }
]
graphs [
 BindToWeaponEvents {nodes [
   n1 = BindToWeaponEvents!() -> (then, DoBind:bool, EquippedWeapon:Fort Weapon Object Reference) {is=title K2Node_FunctionEntry}
   > Parent: BindToWeaponEvents!(execute, DoBind:bool=@n1, EquippedWeapon:Fort Weapon Object Reference=@n1.EquippedWeapon) -> (then) {is=title K2Node_CallParentFunction}
  ]}
 EventGraph {nodes [
   Event Construct()
   > Parent: Construct()
   > n1 = GetDynamicMaterial(self=Get Image_Reticle_Pips() -> (Image_Reticle_Pips:Image Object Reference))
   > Set M_ReticlePips(M_ReticlePips:Material Instance Dynamic Object Reference=@n1) -> (Output_Get:Material Instance Dynamic Object Reference)
   > UpdateCooldownQueryTimer(self:Fort Weapon Reticle Extension Widget Base Object Reference)
   n2 = GetOwningPlayer()
   > Is Valid(InputObject=@n2)
   > n3 = Cast To PlayerPawn_Athena(Object=GetViewTarget!(self:Controller Object Reference=@n2) -> (ReturnValue:Actor Object Reference))
   > Set ViewTargetPlayerPawn(ViewTargetPlayerPawn:Player Pawn Athena Object Reference=@n3) -> (Output_Get:Player Pawn Athena Object Reference)
   > n4 = Set EquippedWeapon(EquippedWeapon:Fort Weapon Object Reference=GetEquippedWeapon!(self:Fort Weapon Reticle Extension Widget Base Object Reference) -> (ReturnValue:Fort Weapon Object Reference)) -> (Output_Get:Fort Weapon Object Reference)
   n5 = GetLocalMagazineAmmoCount!(self:Fort Weapon Object Reference=@n4) -> (ReturnValue:int=0)
   > Set LocalCount Old(LocalCount Old:int=@n5) -> (Output_Get:int=0)
   n6 = Get M_ReticlePips() -> (M_ReticlePips:Material Instance Dynamic Object Reference)
   > SetScalarParameterValue(self=@n6, ParameterName=CurrentPip, Value="To Float (Integer)"(InInt=@n5))
   > SetScalarParameterValue(
    self=@n6
    ParameterName=TotalPips
    Value="To Float (Integer)"(InInt=GetBulletsPerClip!(self:Fort Weapon Object Reference=@n4) -> (ReturnValue:int=0))
   )
   > n7 = Set ReadyForUse(ReadyForUse:bool="integer > integer"(A=GetLocalRemainingAmmo!(self:Fort Weapon Object Reference=@n4) -> (ReturnValue:int=0))) -> (Output_Get:bool)
   > n8 = Branch(Condition=@n7)
   n9 = Get Anim_ReadyForUse() -> (Anim_ReadyForUse:Widget Animation Object Reference)
   > PlayAnimationForward(InAnimation=@n9)
   PlayAnimationReverse(execute=@n8.else, InAnimation=@n9)
   Event Destruct()
   > HandleCooldownTimer!(execute, self:Fort Weapon Reticle Extension Widget Base Object Reference, bActivateTimer:bool=false, TimerInterval:float=0) -> (then)
   > Parent: Destruct()
   n10 = Event OnLocalAmmoChanged!() -> (OutputDelegate, then, LocalCount:int=0, LocalRemaining:int=0) {is=title K2Node_Event}
   > Parent: OnLocalAmmoChanged!(execute, LocalCount:int=@n10.LocalCount, LocalRemaining:int=@n10.LocalRemaining) -> (then) {is=title K2Node_CallParentFunction}
   > n11 = Set Local Count(Local Count:int=@n10.LocalCount) -> (Output_Get:int=0)
   > SetScalarParameterValue(self=Get M_ReticlePips(), ParameterName=CurrentPip, Value="To Float (Integer)"(InInt=@n11))
   > n12 = Sequence!(execute) -> (then_0, then_1, then_2:Exec)
   > UpdateCooldownQueryTimer()
   > n13 = Set ReadyForUse(ReadyForUse="integer > integer"(A=@n10.LocalRemaining))
   > n14 = Branch(Condition=@n13)
   n15 = Get Anim_ReadyForUse()
   > PlayAnimationForward(InAnimation=@n15)
   PlayAnimationReverse(execute=@n14.else, InAnimation=@n15)
   Branch(
    execute=@n12.then_1
    Condition="integer > integer"(A=Get Local Count() -> (Local Count:int=0), B=Get LocalCount Old() -> (LocalCount Old:int=0))
   )
   > PlayAnimation(InAnimation=Get Anim_PipReady() -> (Anim_PipReady:Widget Animation Object Reference))
   > PlaySound2D(Sound=/MotoWrestleGameplay/Assets/Audio/IP_SC_MotoWrestle_Player_Recharge)
   Set LocalCount Old(execute=@n12.then_2, LocalCount Old=Get Local Count() -> (Local Count=0)) -> (Output_Get=0)
   Event UpdateRemainingCooldown() {is=title K2Node_Event}
   > n16 = GetRemainingCooldownTimer!(execute, self) -> (then, RemainingTime:double, TotalDuration:double)
   > SetScalarParameterValue(self=Get M_ReticlePips(), ParameterName=Progress, Value=float / float(A=@n16, B=@n16.TotalDuration))
   Event UpdateCooldownQueryTimer() {is=title K2Node_Event}
   > n17 = Get!(execute:Exec) -> (then:Exec, else:Exec, EquippedWeapon:Fort Weapon Object Reference)
   > n18 = Branch(Condition="integer >= integer"(A=GetLocalMagazineAmmoCount(self=@n17) -> (ReturnValue=0), B=GetBulletsPerClip(self=@n17) -> (ReturnValue=0)))
   > HandleCooldownTimer(bActivateTimer=false, TimerInterval=0)
   HandleCooldownTimer(execute=@n18.else, bActivateTimer=true, TimerInterval=0.1)
  ]}
 GetPipCollection {nodes [
   GetPipCollection() {is=title K2Node_FunctionEntry}
   > Return Node!(execute, PipCollection:WBP Reticle Recharge Stack Base Pip Collection Object Reference)
  ]}
 GetRemainingCooldownTimer {nodes [
   GetRemainingCooldownTimer() {is=title K2Node_FunctionEntry}
   > n1 = Cast To FortPlayerController(Object=GetOwningPlayer())
   > n2 = Get()
   n3 = GetCoolDownTimeRemaining!(
    self:Fort Item Object Reference=GetInventoryItem!(self:Fort Weapon Object Reference=@n2) -> (ReturnValue:Fort Item Object Reference)
    FortPC:Fort Player Controller Object Reference=@n1
   ) -> (CooldownDuration:float, ReturnValue:float)
   > Return Node!(execute, RemainingTime:double=@n3.ReturnValue, TotalDuration:double=@n3)
   Return Node(execute=[@n2.else, @n1.CastFailed], RemainingTime=0, TotalDuration=0)
  ]}
 OnTargetInRangeChanged {nodes [
   n1 = Branch(Condition=AND Boolean(A=OnTargetInRangeChanged!() -> (then, bValid:bool=~) {is=title K2Node_FunctionEntry}, B=Get ReadyForUse() -> (ReadyForUse:bool)))
   n2 = Get Anim_ReadyForUse()
   > PlayAnimationForward(InAnimation=@n2)
   PlayAnimationReverse(execute=@n1.else, InAnimation=@n2)
  ]}
]
variables [
 EquippedWeapon {type=FortWeapon, default=null}
 Local Count {type=integer}
 LocalCount Old {type=integer}
 M_ReticlePips {type=MaterialInstanceDynamic}
 ReadyForUse {type=boolean}
 ViewTargetPlayerPawn {type=PlayerPawn_Athena_C}
]
widgets {
 info {
  inheritedWidgetCount=1
  parentClass=/Game/Athena/HUD/reticles/ReticleExtensionWidgets/ChargeStack/WBP_Reticle_RechargeStackBase.WBP_Reticle_RechargeStackBase_C
  rootWidgetClass=&Overlay
  widgetCount=4
 }
 widgets [
  Overlay_56 {widgetClassPath=&Overlay, widget=Overlay_56}
  Image_Reticle {
   widgetClassPath=&Image
   bIsVariable=true
   parent=Overlay_56
   widget=Image_Reticle
   properties {brush {
     bIsDynamicallyLoaded=false
     drawAs=Image
     imageType=FullColor
     mirroring=NoMirror
     resourceName=None
     resourceObject=/MotoWrestleGameplay/UI/Materials/MI_UI_Ret_MotoWrestle_V2
     tiling=NoTile
     imageSize=<100 100>
     margin=<ltrb 0 0 0 0>
     outlineSettings {
      bUseBrushTransparency=false
      roundingType=HalfHeightRadius
      width=0
      color {colorUseRule=UseColor_Specified, specifiedColor=<rgba 0 0 0 0>}
      cornerRadii=<0 0 0 0>
     }
     tintColor {colorUseRule=UseColor_Specified, specifiedColor=<rgba 1 1 1 1>}
     uVRegion {bIsValid=false, max=<0 0>, min=<0 0>}
    }}
   slot {class=OverlaySlot, properties {horizontalAlignment=HAlign_Center, verticalAlignment=VAlign_Center}}
  }
  Image_Reticle_Pips {
   widgetClassPath=&Image
   bIsVariable=true
   parent=Overlay_56
   widget=Image_Reticle_Pips
   properties {
    brush {
     bIsDynamicallyLoaded=false
     drawAs=Image
     imageType=FullColor
     mirroring=NoMirror
     resourceName=None
     resourceObject=/MotoWrestleGameplay/UI/Materials/M_UI_Pip_MotoWrestle
     tiling=NoTile
     imageSize=<120 120>
     margin=<ltrb 0 0 0 0>
     outlineSettings {
      bUseBrushTransparency=false
      roundingType=HalfHeightRadius
      width=0
      color {colorUseRule=UseColor_Specified, specifiedColor=<rgba 0 0 0 0>}
      cornerRadii=<0 0 0 0>
     }
     tintColor {colorUseRule=UseColor_Specified, specifiedColor=<rgba 1 1 1 1>}
     uVRegion {bIsValid=false, max=<0 0>, min=<0 0>}
    }
    renderTransform {angle=0, scale=<1 1>, shear=<0 0>, translation=<40 20>}
   }
   slot {class=OverlaySlot, properties {horizontalAlignment=HAlign_Center, verticalAlignment=VAlign_Center}}
  }
  TutorialHighlight {widgetClassPath=&TutorialHighlight}
 ]
}
